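# Physics asset for the AnvilAlarmRobin backpack cosmetic, defining jiggle physics for its standing hair strands and right-side zipper.
# Small capsules on base bones serve as roots for larger capsules on the dyn bones.
# Left and right A and B strands form two-segment chains with looser swing at the tips, while front strands have one segment.
# Narrow swing and twist limits keep hair stiff; the zipper swings mainly on one axis with twist locked.
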PhysicsAsset /BRCosmetics/Accessories/FORT_Backpacks/Backpack_M_MED_AnvilAlarmRobin/Meshes/M_MED_AnvilAlarmRobin_Pack_Physics
bodies {
 base_hairStandFront_l [{shapeType=Capsule, length=0, radius=0.5}]
 base_hairStandFront_r [{shapeType=Capsule, length=0.8146, radius=0.5}]
 base_hairStand_A_01_l [{shapeType=Capsule, length=0, radius=0.5315}]
 base_hairStand_A_01_r [{shapeType=Capsule, length=0.505, radius=0.505, shapeName=base_hairStand_A_01_r_capsule, rotation=<rot 0 -90 0>}]
 base_hairStand_B_01_l [{shapeType=Capsule, length=0, radius=0.5}]
 base_hairStand_B_01_r [{shapeType=Capsule, length=0.505, radius=0.505, shapeName=base_hairStand_B_01_r_capsule, rotation=<rot 0 -90 0>}]
 base_zipper_r [{shapeType=Capsule, length=0, radius=0.5, shapeName=base_zipper_r_capsule, rotation=<rot 0 -90 0>}]
 dyn_hairStandFront_l [{shapeType=Capsule, length=6, radius=4, center=<3 0 0>, rotation=<rot -90 0 0>}]
 dyn_hairStandFront_r [{shapeType=Capsule, length=7, radius=5, center=<2 0 0>, rotation=<rot -90 0 0>}]
 dyn_hairStand_A_01_l [{shapeType=Capsule, length=6, radius=4, center=<1 0 0>, rotation=<rot -90 0 0>}]
 dyn_hairStand_A_01_r [{shapeType=Capsule, length=6, radius=5, center=<3.5 0 0>, rotation=<rot -90 0 0>}]
 dyn_hairStand_A_02_l [{shapeType=Capsule, length=6, radius=4, center=<1 0 0>, rotation=<rot -90 0 0>}]
 dyn_hairStand_A_02_r [{shapeType=Capsule, length=6, radius=5, center=<3.5 0 0>, rotation=<rot -90 0 0>}]
 dyn_hairStand_B_01_l [{shapeType=Capsule, length=4, radius=2, center=<2.5 0 0>, rotation=<rot -90 0 0>}]
 dyn_hairStand_B_01_r [{shapeType=Capsule, length=6, radius=3, center=<3 0 0>, rotation=<rot -87.4133 0 0>}]
 dyn_hairStand_B_02_l [{shapeType=Capsule, length=4, radius=2, center=<2.5 0 0>, rotation=<rot -90 0 0>}]
 dyn_hairStand_B_02_r [{shapeType=Capsule, length=6, radius=3, center=<3 0 0>, rotation=<rot -87.4133 0 0>}]
 dyn_zipper_r [{
   shapeType=Capsule
   length=1.3154
   radius=0.6811
   shapeName=dyn_zipper_r_capsule
   center=<3.083 0.0008 0.0004>
   rotation=<rot 0 90.0857 -89.9658>
  }]
}
constraints [
 {bone1Name=dyn_hairStand_A_02_r, bone2Name=dyn_hairStand_A_01_r, swing1LimitDegrees=6, swing2LimitDegrees=6, twistLimitDegrees=2}
 {bone1Name=dyn_hairStand_A_01_r, bone2Name=base_hairStand_A_01_r, swing1LimitDegrees=4, swing2LimitDegrees=4, twistLimitDegrees=0}
 {bone1Name=dyn_hairStand_A_02_l, bone2Name=dyn_hairStand_A_01_l, swing1LimitDegrees=11, swing2LimitDegrees=11, twistLimitDegrees=2}
 {bone1Name=dyn_hairStand_A_01_l, bone2Name=base_hairStand_A_01_l, swing1LimitDegrees=4, swing2LimitDegrees=4, twistLimitDegrees=0}
 {bone1Name=dyn_hairStand_B_01_l, bone2Name=base_hairStand_B_01_l, swing1LimitDegrees=4, swing2LimitDegrees=4, twistLimitDegrees=0}
 {bone1Name=dyn_hairStand_B_02_l, bone2Name=dyn_hairStand_B_01_l, swing1LimitDegrees=10, swing2LimitDegrees=10, twistLimitDegrees=2}
 {bone1Name=dyn_hairStand_B_02_r, bone2Name=dyn_hairStand_B_01_r, swing1LimitDegrees=12, swing2LimitDegrees=12, twistLimitDegrees=0}
 {bone1Name=dyn_hairStand_B_01_r, bone2Name=base_hairStand_B_01_r, swing1LimitDegrees=4, swing2LimitDegrees=4, twistLimitDegrees=0}
 {bone1Name=dyn_hairStandFront_l, bone2Name=base_hairStandFront_l, swing1LimitDegrees=8, swing2LimitDegrees=8, twistLimitDegrees=2}
 {bone1Name=dyn_hairStandFront_r, bone2Name=base_hairStandFront_r, swing1LimitDegrees=12, swing2LimitDegrees=12, twistLimitDegrees=2}
 {bone1Name=dyn_zipper_r, bone2Name=base_zipper_r, swing1LimitDegrees=5, swing2LimitDegrees=45, twistLimitDegrees=45, twistMotion=Locked}
]
