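# Anim Blueprint for the AnvilAlarmRobin backpack cosmetic, adding secondary physics motion to its hairStand bones.
# The input pose gets a pack-bone offset and Anim Dynamics on C/D/E bones, then is cached.
# bIsInFrontEnd selects further Anim Dynamics on A/B/Front bones in the front end, otherwise a RigidBody simulation relative to pack.
# The event graph updates Gravity Override and AnimDynAlpha, though the AnimGraph shows neither variable wired in.
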AnimBlueprint /BRCosmetics/Accessories/FORT_Backpacks/Backpack_M_MED_AnvilAlarmRobin/Meshes/M_MED_AnvilAlarmRobin_Pack_AnimBP
target_skeleton=&/M_MED_AnvilAlarmRobin_Pack_Skeleton
graphs [
 AnimGraph {nodes [
   n1 = RigidBody!(ComponentPose=Local To Component(LocalPose=Use cached pose 'PreFrontEnd_1'()), Alpha, OverrideWorldGravity:Vector="0,0,0", ExternalForce) -> (Pose) {
    bOverrideWorldGravity=true
    overridePhysicsAsset=&/M_MED_AnvilAlarmRobin_Pack_Physics
    simulationSpace=BaseBoneSpace
    baseBoneRef {boneName=pack}
    simSpaceSettings {
     dampingAlpha=0
     maxAngularAcceleration=21
     maxAngularVelocity=8
     maxLinearAcceleration=1500
     maxLinearVelocity=700
     worldAlpha=1
     externalLinearDragV=<0.75 0.75 0.75>
    }
   }
   n2 = Anim Dynamics - Bone: dyn_hairStand_A_01_r!(
    ComponentPose=Local To Component(LocalPose=Use cached pose 'PreFrontEnd_1'())
    Alpha=1
    GravityOverride:Vector="0,0,-980"
    ExternalForce="0,0,0"
   ) -> (Pose) {
    angularDampingOverride=1
    angularSpringConstant=35000
    bAngularSpring=true
    bOverrideAngularDamping=true
    bUseGravityOverride=true
    gravityScale=0.35
    simulationSpace=BoneRelative
    boundBone {boneName=dyn_hairStand_A_01_r}
    componentAppliedLinearAccClamp=<1200 1200 1200>
    componentLinearAccScale=<0.1 0.1 0.1>
    componentLinearVelScale=<0.1 0.1 0.1>
    gravityOverride {z=-980}
    physicsBodyDefinitions [{
      boundBone {boneName=dyn_hairStand_A_01_r}
      boxExtents=<30 30 30>
      constraintSetup {angularLimitsMax=<3 0 15>, angularLimitsMin=<-3 -10 -25>, angularTarget=<1 0 0>}
      localJointOffset=<20 0 0>
     }]
    relativeSpaceBone {boneName=pack}
    simSpaceSettings {maxAngularAcceleration=0.4, maxAngularVelocity=0.1}
   }
   n3 = Anim Dynamics - Bone: dyn_hairStand_A_02_r!(ComponentPose=@n2, Alpha=1, GravityOverride:Vector="0,0,-980", ExternalForce="0,0,0") -> (Pose) {
    angularDampingOverride=1
    angularSpringConstant=20000
    bAngularSpring=true
    bOverrideAngularDamping=true
    bUseGravityOverride=true
    gravityScale=0.35
    simulationSpace=BoneRelative
    boundBone {boneName=dyn_hairStand_A_02_r}
    componentAppliedLinearAccClamp=<1200 1200 1200>
    componentLinearAccScale=<0.1 0.1 0.1>
    componentLinearVelScale=<0.1 0.1 0.1>
    gravityOverride {z=-980}
    physicsBodyDefinitions [{
      boundBone {boneName=dyn_hairStand_A_02_r}
      boxExtents=<30 30 30>
      constraintSetup {angularLimitsMax=<3 10 15>, angularLimitsMin=<-3 -10 -25>, angularTarget=<1 0 0>}
      localJointOffset=<20 0 0>
     }]
    relativeSpaceBone {boneName=pack}
    simSpaceSettings {maxAngularAcceleration=0.4, maxAngularVelocity=0.1}
   }
   n4 = Anim Dynamics - Bone: dyn_hairStand_B_01_r!(ComponentPose=@n3, Alpha=1, GravityOverride:Vector="0,0,-980", ExternalForce="0,0,0") -> (Pose) {
    angularDampingOverride=1
    angularSpringConstant=15000
    bAngularSpring=true
    bOverrideAngularDamping=true
    bUseGravityOverride=true
    gravityScale=0.35
    simulationSpace=BoneRelative
    boundBone {boneName=dyn_hairStand_B_01_r}
    componentAppliedLinearAccClamp=<1200 1200 1200>
    componentLinearAccScale=<0.1 0.1 0.1>
    componentLinearVelScale=<0.1 0.1 0.1>
    gravityOverride {z=-980}
    physicsBodyDefinitions [{
      boundBone {boneName=dyn_hairStand_B_01_r}
      boxExtents=<25 25 25>
      constraintSetup {angularLimitsMax=<3 0 15>, angularLimitsMin=<-3 -10 -25>, angularTarget=<1 0 0>}
      localJointOffset=<15 0 0>
     }]
    relativeSpaceBone {boneName=pack}
    simSpaceSettings {maxAngularAcceleration=0.4, maxAngularVelocity=0.1}
   }
   n5 = Anim Dynamics - Bone: dyn_hairStand_B_02_r!(ComponentPose=@n4, Alpha=1, GravityOverride:Vector="0,0,-980", ExternalForce="0,0,0") -> (Pose) {
    angularDampingOverride=1
    angularSpringConstant=10000
    bAngularSpring=true
    bOverrideAngularDamping=true
    bUseGravityOverride=true
    gravityScale=0.35
    simulationSpace=BoneRelative
    boundBone {boneName=dyn_hairStand_B_02_r}
    componentAppliedLinearAccClamp=<1200 1200 1200>
    componentLinearAccScale=<0.1 0.1 0.1>
    componentLinearVelScale=<0.1 0.1 0.1>
    gravityOverride {z=-980}
    physicsBodyDefinitions [{
      boundBone {boneName=dyn_hairStand_B_02_r}
      boxExtents=<25 25 25>
      constraintSetup {angularLimitsMax=<3 10 15>, angularLimitsMin=<-3 -10 -25>, angularTarget=<1 0 0>}
      localJointOffset=<15 0 0>
     }]
    relativeSpaceBone {boneName=pack}
    simSpaceSettings {maxAngularAcceleration=0.4, maxAngularVelocity=0.1}
   }
   n6 = Anim Dynamics - Bone: dyn_hairStandFront_r!(ComponentPose=@n5, Alpha=1, GravityOverride:Vector="0,0,-980", ExternalForce="0,0,0") -> (Pose) {
    angularDampingOverride=1
    angularSpringConstant=30000
    bAngularSpring=true
    bOverrideAngularDamping=true
    bUseGravityOverride=true
    gravityScale=0.35
    simulationSpace=BoneRelative
    boundBone {boneName=dyn_hairStandFront_r}
    componentAppliedLinearAccClamp=<1200 1200 1200>
    componentLinearAccScale=<0.5 0.5 0.5>
    componentLinearVelScale=<0.5 0.5 0.5>
    gravityOverride {z=-980}
    physicsBodyDefinitions [{
      boundBone {boneName=dyn_hairStandFront_r}
      boxExtents=<35 40 40>
      constraintSetup {angularLimitsMax=<3 0 0>, angularLimitsMin=<-3 -30 -30>, angularTarget=<1 0 0>}
      localJointOffset=<30 0 0>
     }]
    relativeSpaceBone {boneName=pack}
    simSpaceSettings {maxAngularAcceleration=0.4, maxAngularVelocity=0.1}
   }
   n7 = Anim Dynamics - Bone: dyn_hairStand_A_01_l!(ComponentPose=@n6, Alpha=1, GravityOverride:Vector="0,0,-980", ExternalForce="0,0,0") -> (Pose) {
    angularDampingOverride=1
    angularSpringConstant=35000
    bAngularSpring=true
    bOverrideAngularDamping=true
    bUseGravityOverride=true
    gravityScale=0.35
    simulationSpace=BoneRelative
    boundBone {boneName=dyn_hairStand_A_01_l}
    componentAppliedLinearAccClamp=<1200 1200 1200>
    componentLinearAccScale=<0.1 0.1 0.1>
    componentLinearVelScale=<0.1 0.1 0.1>
    gravityOverride {z=-980}
    physicsBodyDefinitions [{
      boundBone {boneName=dyn_hairStand_A_01_l}
      boxExtents=<30 30 30>
      constraintSetup {angularLimitsMax=<3 0 15>, angularLimitsMin=<-3 -10 -25>, angularTarget=<1 0 0>}
      localJointOffset=<20 0 0>
     }]
    relativeSpaceBone {boneName=pack}
    simSpaceSettings {maxAngularAcceleration=0.4, maxAngularVelocity=0.1}
   }
   n8 = Anim Dynamics - Bone: dyn_hairStand_A_02_l!(ComponentPose=@n7, Alpha=1, GravityOverride:Vector="0,0,-980", ExternalForce="0,0,0") -> (Pose) {
    angularDampingOverride=1
    angularSpringConstant=20000
    bAngularSpring=true
    bOverrideAngularDamping=true
    bUseGravityOverride=true
    gravityScale=0.35
    simulationSpace=BoneRelative
    boundBone {boneName=dyn_hairStand_A_02_l}
    componentAppliedLinearAccClamp=<1200 1200 1200>
    componentLinearAccScale=<0.1 0.1 0.1>
    componentLinearVelScale=<0.1 0.1 0.1>
    gravityOverride {z=-980}
    physicsBodyDefinitions [{
      boundBone {boneName=dyn_hairStand_A_02_l}
      boxExtents=<30 30 30>
      constraintSetup {angularLimitsMax=<3 10 15>, angularLimitsMin=<-3 -10 -25>, angularTarget=<1 0 0>}
      localJointOffset=<20 0 0>
     }]
    relativeSpaceBone {boneName=pack}
    simSpaceSettings {maxAngularAcceleration=0.4, maxAngularVelocity=0.1}
   }
   n9 = Anim Dynamics - Bone: dyn_hairStand_B_01_l!(ComponentPose=@n8, Alpha=1, GravityOverride:Vector="0,0,-980", ExternalForce="0,0,0") -> (Pose) {
    angularDampingOverride=1
    angularSpringConstant=15000
    bAngularSpring=true
    bOverrideAngularDamping=true
    bUseGravityOverride=true
    gravityScale=0.35
    simulationSpace=BoneRelative
    boundBone {boneName=dyn_hairStand_B_01_l}
    componentAppliedLinearAccClamp=<1200 1200 1200>
    componentLinearAccScale=<0.1 0.1 0.1>
    componentLinearVelScale=<0.1 0.1 0.1>
    gravityOverride {z=-980}
    physicsBodyDefinitions [{
      boundBone {boneName=dyn_hairStand_B_01_l}
      boxExtents=<25 25 25>
      constraintSetup {angularLimitsMax=<3 0 15>, angularLimitsMin=<-3 -10 -25>, angularTarget=<1 0 0>}
      localJointOffset=<15 0 0>
     }]
    relativeSpaceBone {boneName=pack}
    simSpaceSettings {maxAngularAcceleration=0.4, maxAngularVelocity=0.1}
   }
   n10 = Anim Dynamics - Bone: dyn_hairStand_B_02_l!(ComponentPose=@n9, Alpha=1, GravityOverride:Vector="0,0,-980", ExternalForce="0,0,0") -> (Pose) {
    angularDampingOverride=1
    angularSpringConstant=10000
    bAngularSpring=true
    bOverrideAngularDamping=true
    bUseGravityOverride=true
    gravityScale=0.35
    simulationSpace=BoneRelative
    boundBone {boneName=dyn_hairStand_B_02_l}
    componentAppliedLinearAccClamp=<1200 1200 1200>
    componentLinearAccScale=<0.1 0.1 0.1>
    componentLinearVelScale=<0.1 0.1 0.1>
    gravityOverride {z=-980}
    physicsBodyDefinitions [{
      boundBone {boneName=dyn_hairStand_B_02_l}
      boxExtents=<25 25 25>
      constraintSetup {angularLimitsMax=<3 10 15>, angularLimitsMin=<-3 -10 -25>, angularTarget=<1 0 0>}
      localJointOffset=<15 0 0>
     }]
    relativeSpaceBone {boneName=pack}
    simSpaceSettings {maxAngularAcceleration=0.4, maxAngularVelocity=0.1}
   }
   n11 = Anim Dynamics - Bone: dyn_hairStandFront_l!(ComponentPose=@n10, Alpha=1, GravityOverride:Vector="0,0,-980", ExternalForce="0,0,0") -> (Pose) {
    angularDampingOverride=1
    angularSpringConstant=30000
    bAngularSpring=true
    bOverrideAngularDamping=true
    bUseGravityOverride=true
    gravityScale=0.35
    simulationSpace=BoneRelative
    boundBone {boneName=dyn_hairStandFront_l}
    componentAppliedLinearAccClamp=<1200 1200 1200>
    componentLinearAccScale=<0.5 0.5 0.5>
    componentLinearVelScale=<0.5 0.5 0.5>
    gravityOverride {z=-980}
    physicsBodyDefinitions [{
      boundBone {boneName=dyn_hairStandFront_l}
      boxExtents=<35 40 40>
      constraintSetup {angularLimitsMax=<3 30 0>, angularLimitsMin=<-3 0 -30>, angularTarget=<1 0 0>}
      localJointOffset=<30 0 0>
     }]
    relativeSpaceBone {boneName=pack}
    simSpaceSettings {maxAngularAcceleration=0.4, maxAngularVelocity=0.1}
   }
   Output Pose(Result=Blend Poses by bool(BlendPose_0=Component To Local(ComponentPose=@n11), BlendPose_1=Component To Local(ComponentPose=@n1), bActiveValue=Get bIsInFrontEnd())) {layerGroup=None}
   n12 = Input Pose()
   n13 = Anim Dynamics - Bone: dyn_hairStand_C_01_r!(
    ComponentPose="Transform (Modify) Bone - Bone: pack"(ComponentPose=Local To Component(LocalPose=@n12), Translation="0,-3,0", Rotation="0,0,0") {translationMode=BMM_Additive, boneToModify {boneName=pack}, translation {y=-3}}
    Alpha=1
    GravityOverride:Vector="0,0,-980"
    ExternalForce="0,0,0"
   ) -> (Pose) {
    angularDampingOverride=1
    angularSpringConstant=10000
    bOverrideAngularDamping=true
    bUseGravityOverride=true
    gravityScale=0.35
    simulationSpace=BoneRelative
    boundBone {boneName=dyn_hairStand_C_01_r}
    componentAppliedLinearAccClamp=<1200 1200 1200>
    componentLinearAccScale=<0.25 0.25 0.25>
    componentLinearVelScale=<0.25 0.25 0.25>
    gravityOverride {z=-980}
    physicsBodyDefinitions [{
      boundBone {boneName=dyn_hairStand_C_01_r}
      boxExtents=<20 30 30>
      constraintSetup {angularLimitsMax=<3 10 0>, angularLimitsMin=<-3 -10 -25>, angularTarget=<1 0 0>}
      localJointOffset=<15 0 0>
     }]
    relativeSpaceBone {boneName=pack}
   }
   n14 = Anim Dynamics - Bone: dyn_hairStand_D_01_r!(ComponentPose=@n13, Alpha=1, GravityOverride:Vector="0,0,-980", ExternalForce="0,0,0") -> (Pose) {
    angularDampingOverride=1
    angularSpringConstant=30000
    bAngularSpring=true
    bOverrideAngularDamping=true
    bUseGravityOverride=true
    gravityScale=0.35
    simulationSpace=BoneRelative
    boundBone {boneName=dyn_hairStand_D_01_r}
    componentAppliedLinearAccClamp=<1200 1200 1200>
    componentLinearAccScale=<0.5 0.5 0.5>
    componentLinearVelScale=<0.5 0.5 0.5>
    gravityOverride {z=-980}
    physicsBodyDefinitions [{
      boundBone {boneName=dyn_hairStand_D_01_r}
      boxExtents=<30 35 35>
      constraintSetup {angularLimitsMax=<3 25 5>, angularLimitsMin=<-3 -25 -25>, angularTarget=<1 0 0>}
      localJointOffset=<25 0 0>
     }]
    relativeSpaceBone {boneName=pack}
   }
   n15 = Anim Dynamics - Bone: dyn_hairStand_E_01_r!(ComponentPose=@n14, Alpha=1, GravityOverride:Vector="0,0,-980", ExternalForce="0,0,0") -> (Pose) {
    angularDampingOverride=1
    angularSpringConstant=10000
    bOverrideAngularDamping=true
    bUseGravityOverride=true
    gravityScale=0.35
    simulationSpace=BoneRelative
    boundBone {boneName=dyn_hairStand_E_01_r}
    componentAppliedLinearAccClamp=<1200 1200 1200>
    componentLinearAccScale=<0.35 0.35 0.35>
    componentLinearVelScale=<0.35 0.35 0.35>
    gravityOverride {z=-980}
    physicsBodyDefinitions [{
      boundBone {boneName=dyn_hairStand_E_01_r}
      boxExtents=<20 30 30>
      constraintSetup {angularLimitsMax=<3 10 0>, angularLimitsMin=<-3 -10 -25>, angularTarget=<1 0 0>}
      localJointOffset=<17.5 0 0>
     }]
    relativeSpaceBone {boneName=pack}
   }
   n16 = Anim Dynamics - Bone: dyn_hairStand_C_01_l!(ComponentPose=@n15, Alpha=1, GravityOverride:Vector="0,0,-980", ExternalForce="0,0,0") -> (Pose) {
    angularDampingOverride=1
    angularSpringConstant=10000
    bOverrideAngularDamping=true
    bUseGravityOverride=true
    gravityScale=0.35
    simulationSpace=BoneRelative
    boundBone {boneName=dyn_hairStand_C_01_l}
    componentAppliedLinearAccClamp=<1200 1200 1200>
    componentLinearAccScale=<0.25 0.25 0.25>
    componentLinearVelScale=<0.25 0.25 0.25>
    gravityOverride {z=-980}
    physicsBodyDefinitions [{
      boundBone {boneName=dyn_hairStand_C_01_l}
      boxExtents=<20 30 30>
      constraintSetup {angularLimitsMax=<3 10 0>, angularLimitsMin=<-3 -10 -25>, angularTarget=<1 0 0>}
      localJointOffset=<15 0 0>
     }]
    relativeSpaceBone {boneName=pack}
   }
   n17 = Anim Dynamics - Bone: dyn_hairStand_D_01_l!(ComponentPose=@n16, Alpha=1, GravityOverride:Vector="0,0,-980", ExternalForce="0,0,0") -> (Pose) {
    angularDampingOverride=1
    angularSpringConstant=30000
    bAngularSpring=true
    bOverrideAngularDamping=true
    bUseGravityOverride=true
    gravityScale=0.35
    simulationSpace=BoneRelative
    boundBone {boneName=dyn_hairStand_D_01_l}
    componentAppliedLinearAccClamp=<1200 1200 1200>
    componentLinearAccScale=<0.5 0.5 0.5>
    componentLinearVelScale=<0.5 0.5 0.5>
    gravityOverride {z=-980}
    physicsBodyDefinitions [{
      boundBone {boneName=dyn_hairStand_D_01_l}
      boxExtents=<30 35 35>
      constraintSetup {angularLimitsMax=<3 10 0>, angularLimitsMin=<-3 -25 -25>, angularTarget=<1 0 0>}
      localJointOffset=<25 0 0>
     }]
    relativeSpaceBone {boneName=pack}
   }
   n18 = Anim Dynamics - Bone: dyn_hairStand_E_01_l!(ComponentPose=@n17, Alpha=1, GravityOverride:Vector="0,0,-980", ExternalForce="0,0,0") -> (Pose) {
    angularDampingOverride=1
    angularSpringConstant=10000
    bOverrideAngularDamping=true
    bUseGravityOverride=true
    gravityScale=0.35
    simulationSpace=BoneRelative
    boundBone {boneName=dyn_hairStand_E_01_l}
    componentAppliedLinearAccClamp=<1200 1200 1200>
    componentLinearAccScale=<0.35 0.35 0.35>
    componentLinearVelScale=<0.35 0.35 0.35>
    gravityOverride {z=-980}
    physicsBodyDefinitions [{
      boundBone {boneName=dyn_hairStand_E_01_l}
      boxExtents=<20 30 30>
      constraintSetup {angularLimitsMax=<3 10 0>, angularLimitsMin=<-3 -10 -25>, angularTarget=<1 0 0>}
      localJointOffset=<17.5 0 0>
     }]
    relativeSpaceBone {boneName=pack}
   }
   Save cached pose 'PreFrontEnd_1'(Pose=Component To Local(ComponentPose=@n18)) {cacheName=PreFrontEnd_1}
  ]}
 EventGraph {nodes [
   Event BlueprintUpdateAnimation()
   > n1 = "Gravity Override State Function (Parameter Asset) (Nativized)"(
    CharacterPartAnimInstance=Self-Reference()
    GameTimeInSeconds=GetGameTimeInSeconds()
    Parameters=&/M_MED_AnvilAlarmRobin_Parameters
   )
   > Set Gravity Override(Gravity Override=@n1)
   > Set AnimDynAlpha(AnimDynAlpha:double=SelectFloat(A=0.5, B=1, bPickA=Get bIsInFrontEnd())) -> (Output_Get:double)
   TryGetPawnOwner()
  ]}
]
variables [
 AnimDynAlpha {type=number}
 Current State Out {type=string, default=None}
 Gravity Override {type=Vector}
 Sim Space Settings {
  type=SimSpaceSettings
  default {
   dampingAlpha=0
   maxAngularAcceleration=21
   maxAngularVelocity=7
   maxLinearAcceleration=2500
   maxLinearVelocity=700
   velocityScaleZ=0.55
   worldAlpha=0.55
   externalLinearDragV=<0.75 0.75 0.75>
  }
 }
]
